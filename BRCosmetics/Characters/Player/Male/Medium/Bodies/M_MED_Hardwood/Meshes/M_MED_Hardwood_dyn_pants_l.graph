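# Physics asset for the left-leg shorts "dyn" bones on the male medium Hardwood player body, with capsule bodies for front, back, inner and outer flaps plus the thigh.
# Each flap hangs from a tiny MSTR capsule through tight angular limits, with the back using a two-segment chain.
# Inner and outer MSTR bodies join the thigh twist bone with free rotation; front and back MSTR bodies have no thigh constraint.
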PhysicsAsset /BRCosmetics/Characters/Player/Male/Medium/Bodies/M_MED_Hardwood/Meshes/M_MED_Hardwood_dyn_pants_l
bodies {
 dyn_shorts_bk_01_l [{shapeType=Capsule, length=10.25, radius=4.75, center=<5 0 3>, rotation=<rot 0 0 90>}]
 dyn_shorts_bk_02_l [{shapeType=Capsule, length=9.5, radius=5.5, center=<5 0 4>, rotation=<rot 0 0 90>}]
 dyn_shorts_bk_l_MSTR [{shapeType=Capsule, length=1, radius=1}]
 dyn_shorts_ft_l [{shapeType=Capsule, length=10.25, radius=4.75, center=<10 0 3>, rotation=<rot 0 0 90>}]
 dyn_shorts_ft_l_MSTR [{shapeType=Capsule, length=1, radius=1}]
 dyn_shorts_in_l [{shapeType=Capsule, length=10.25, radius=4.75, center=<10 0 3>, rotation=<rot 0 0 90>}]
 dyn_shorts_in_l_MSTR [{shapeType=Capsule, length=1, radius=1}]
 dyn_shorts_out_l [{shapeType=Capsule, length=10.75, radius=4.25, center=<10 0 3>, rotation=<rot 0 0 90>}]
 dyn_shorts_out_l_MSTR [{shapeType=Capsule, length=1, radius=1}]
 thigh_twist_01_l [{shapeType=Capsule, length=17, radius=6.75, center=<-2.375 0.1263 0.1081>, rotation=<rot 88.42 0.1455 -180>}]
}
constraints [
 {bone1Name=dyn_shorts_ft_l, bone2Name=dyn_shorts_ft_l_MSTR, swing1LimitDegrees=7, swing2LimitDegrees=15, twistLimitDegrees=2}
 {
  bone1Name=dyn_shorts_out_l_MSTR
  bone2Name=thigh_twist_01_l
  swing1LimitDegrees=45
  swing1Motion=Free
  swing2LimitDegrees=45
  swing2Motion=Free
  twistLimitDegrees=45
  twistMotion=Free
 }
 {bone1Name=dyn_shorts_out_l, bone2Name=dyn_shorts_out_l_MSTR, swing1LimitDegrees=7, swing2LimitDegrees=15, twistLimitDegrees=2}
 {
  bone1Name=dyn_shorts_in_l_MSTR
  bone2Name=thigh_twist_01_l
  swing1LimitDegrees=45
  swing1Motion=Free
  swing2LimitDegrees=45
  swing2Motion=Free
  twistLimitDegrees=45
  twistMotion=Free
 }
 {bone1Name=dyn_shorts_in_l, bone2Name=dyn_shorts_in_l_MSTR, swing1LimitDegrees=7, swing2LimitDegrees=15, twistLimitDegrees=2}
 {bone1Name=dyn_shorts_bk_01_l, bone2Name=dyn_shorts_bk_l_MSTR, swing1LimitDegrees=5, swing2LimitDegrees=5, twistLimitDegrees=5}
 {bone1Name=dyn_shorts_bk_02_l, bone2Name=dyn_shorts_bk_01_l, swing1LimitDegrees=5, swing2LimitDegrees=5, twistLimitDegrees=5}
]
