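# Post-process-style AnimBP for the male medium Hardwood body that copies the parent mesh pose and adds shorts deformation and dynamics.
# Twist and sculpter Control Rigs feed additive shorts-bone offsets, weighted by knee bend measured each update from thigh-to-calf delta yaw.
# Per-leg RigidBody simulations follow, their smoothed alpha dropping while skydiving, before a final right-back shorts correction.
# Some right-side bone modifiers and pelvis-relative thigh calculations are left unconnected.
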AnimBlueprint /BRCosmetics/Characters/Player/Male/Medium/Bodies/M_MED_Hardwood/Meshes/M_MED_Hardwood_AnimBP
target_skeleton=&/M_MED_Hardwood_Skeleton
graphs [
 AnimGraph {nodes [
   n1 = Get calf_rot_r() -> (calf_rot_r:double)
   "Transform (Modify) Bone - Bone: dyn_shorts_bk_r_MSTR"(Alpha=@n1, Translation="-6,0,0", Rotation="0,0,0", Scale="1,1,1") {
    rotationMode=BMM_Additive
    rotationSpace=BCS_BoneSpace
    scaleMode=BMM_Additive
    scaleSpace=BCS_BoneSpace
    translationMode=BMM_Additive
    translationSpace=BCS_BoneSpace
    alphaBoolBlend {blendOption=Linear}
    boneToModify {boneName=dyn_shorts_bk_r_MSTR}
    translation {x=-6}
   }
   n2 = "Transform (Modify) Bone - Bone: dyn_shorts_out_r_MSTR"(Alpha=1, Translation="-6,-0.000,0.000", Rotation="0,0,0", Scale="1,1,1") {
    rotationMode=BMM_Additive
    rotationSpace=BCS_BoneSpace
    translationMode=BMM_Additive
    translationSpace=BCS_BoneSpace
    alphaBoolBlend {blendOption=Linear}
    boneToModify {boneName=dyn_shorts_out_r_MSTR}
    translation {x=-6}
   }
   n3 = "Transform (Modify) Bone - Bone: dyn_shorts_in_r_MSTR"(ComponentPose=@n2, Alpha=1, Translation="-6,-0.000,0.000", Rotation="0,0,0", Scale="1,1,1") {
    rotationMode=BMM_Additive
    rotationSpace=BCS_BoneSpace
    translationMode=BMM_Additive
    translationSpace=BCS_BoneSpace
    alphaBoolBlend {blendOption=Linear}
    boneToModify {boneName=dyn_shorts_in_r_MSTR}
    translation {x=-6}
   }
   "Transform (Modify) Bone - Bone: dyn_shorts_ft_r_MSTR"(ComponentPose=@n3, Alpha=1, Translation="-6,0.000,0.000", Rotation="0,0,0", Scale="1,1,1") {
    rotationMode=BMM_Additive
    rotationSpace=BCS_BoneSpace
    translationMode=BMM_Additive
    translationSpace=BCS_BoneSpace
    alphaBoolBlend {blendOption=Linear}
    boneToModify {boneName=dyn_shorts_ft_r_MSTR}
    translation {x=-6}
   }
   n4 = Get calf_rot_r()
   "Transform (Modify) Bone - Bone: dyn_shorts_bk_02_r"(Alpha=@n4, Translation="0,0,0", Rotation="0,0,0", Scale="1,1,1") {
    rotationMode=BMM_Additive
    rotationSpace=BCS_BoneSpace
    scaleMode=BMM_Additive
    scaleSpace=BCS_BoneSpace
    translationMode=BMM_Additive
    translationSpace=BCS_BoneSpace
    alphaBoolBlend {blendOption=Linear}
    boneToModify {boneName=dyn_shorts_bk_02_r}
   }
   n5 = Control Rig(Source=Copy Pose From Mesh() {bUseAttachedParent=true}) {
    alphaBoolBlend {blendOption=Linear}
    controlRigAssetReference {blueprintRigClass=&/M_MED_Hardwood_Deform_CtrlRig_twists.M_MED_Hardwood_Deform_CtrlRig_twists_C}
   }
   n6 = Control Rig!(
    Source=@n5
    Alpha
    KneeRot_fr_l_114deg:Vector
    KneeTrans_fr_l_114deg:Vector="0.258,-3.591,0.009"
    KneeScale_fr_l_114deg:Vector="1,1,1"
    KneeScale_fr_l_130deg:Vector="1,1,1"
    KneeRot_fr_l_130deg:Vector="0,0,-51.2"
    KneeTrans_fr_l_130deg:Vector="0.258,-3.591,0.009"
   ) -> (Pose) {
    alphaBoolBlend {blendOption=Linear}
    controlRigAssetReference {blueprintRigClass=&/M_MED_Hardwood_Deform_CtrlRig_sculpters.M_MED_Hardwood_Deform_CtrlRig_sculpters_C}
   }
   n7 = Get calf_rot_l() -> (calf_rot_l:double)
   n8 = "Transform (Modify) Bone - Bone: dyn_shorts_bk_l_MSTR"(ComponentPose=Local To Component(LocalPose=@n6), Alpha=@n7, Translation="-10,-0.000,0.000", Rotation="0,0,0", Scale="1,1,1") {
    rotationMode=BMM_Additive
    rotationSpace=BCS_BoneSpace
    translationMode=BMM_Additive
    translationSpace=BCS_BoneSpace
    alphaBoolBlend {blendOption=Linear}
    boneToModify {boneName=dyn_shorts_bk_l_MSTR}
    translation {x=-10}
   }
   n9 = "Transform (Modify) Bone - Bone: dyn_shorts_out_l_MSTR"(ComponentPose=@n8, Alpha=@n7, Translation="-3,-0.000,0.000", Rotation="0,0,0", Scale="1,1,1") {
    rotationMode=BMM_Additive
    rotationSpace=BCS_BoneSpace
    translationMode=BMM_Additive
    translationSpace=BCS_BoneSpace
    alphaBoolBlend {blendOption=Linear}
    boneToModify {boneName=dyn_shorts_out_l_MSTR}
    translation {x=-3}
   }
   n10 = "Transform (Modify) Bone - Bone: dyn_shorts_in_l_MSTR"(ComponentPose=@n9, Alpha=@n7, Translation="-3,-0.000,0.000", Rotation="0,0,0", Scale="1,1,1") {
    rotationMode=BMM_Additive
    rotationSpace=BCS_BoneSpace
    translationMode=BMM_Additive
    translationSpace=BCS_BoneSpace
    alphaBoolBlend {blendOption=Linear}
    boneToModify {boneName=dyn_shorts_in_l_MSTR}
    translation {x=-3}
   }
   n11 = "Transform (Modify) Bone - Bone: dyn_shorts_main_r"(ComponentPose=@n10, Alpha=@n1, Translation="0,0,-6", Rotation="0,0,0", Scale="1,1,1") {
    rotationMode=BMM_Additive
    rotationSpace=BCS_BoneSpace
    translationMode=BMM_Additive
    translationSpace=BCS_BoneSpace
    alphaBoolBlend {blendOption=Linear}
    boneToModify {boneName=dyn_shorts_main_r}
    translation {z=-6}
   }
   n12 = Get Is Sky Diving() -> (Is Sky Diving:double)
   n13 = RigidBody!(ComponentPose=@n11, Alpha=@n12, OverrideWorldGravity:Vector="0,0,-980", ExternalForce) -> (Pose) {
    bClampLinearTranslationLimitToRefPose=true
    bOverrideWorldGravity=true
    overridePhysicsAsset=&/M_MED_Hardwood_dyn_pants_l
    simulationSpace=BaseBoneSpace
    alphaBoolBlend {blendOption=Linear}
    alphaScaleBiasClamp {bInterpResult=true, interpSpeedDecreasing=20, interpSpeedIncreasing=20}
    baseBoneRef {boneName=thigh_twist_01_l}
    componentAppliedLinearAccClamp=<1000 1000 1000>
    overrideWorldGravity {z=-980}
    simSpaceSettings {maxAngularAcceleration=100, maxAngularVelocity=5, maxLinearAcceleration=500, maxLinearVelocity=50, worldAlpha=1}
   }
   n14 = RigidBody(ComponentPose=@n13, Alpha=@n12, OverrideWorldGravity="0,0,-980") {
    bClampLinearTranslationLimitToRefPose=true
    bOverrideWorldGravity=true
    overridePhysicsAsset=&/M_MED_Hardwood_dyn_pants_r
    simulationSpace=BaseBoneSpace
    alphaBoolBlend {blendOption=Linear}
    alphaScaleBiasClamp {bInterpResult=true, interpSpeedDecreasing=20, interpSpeedIncreasing=20}
    baseBoneRef {boneName=thigh_twist_01_r}
    componentAppliedLinearAccClamp=<1000 1000 1000>
    overrideWorldGravity {z=-980}
    simSpaceSettings {maxAngularAcceleration=100, maxAngularVelocity=5, maxLinearAcceleration=500, maxLinearVelocity=50, worldAlpha=1}
   }
   n15 = "Transform (Modify) Bone - Bone: dyn_shorts_bk_01_r"(ComponentPose=@n14, Alpha=@n4, Translation="0,0,0", Rotation="0.000,-5,0.000", Scale="0.75,1.000,1.000") {
    rotationMode=BMM_Additive
    rotationSpace=BCS_BoneSpace
    scaleMode=BMM_Additive
    scaleSpace=BCS_BoneSpace
    translationMode=BMM_Additive
    translationSpace=BCS_BoneSpace
    alphaBoolBlend {blendOption=Linear}
    boneToModify {boneName=dyn_shorts_bk_01_r}
    rotation {yaw=-5}
    scale {x=0.75}
   }
   Output Pose(Result=Component To Local(ComponentPose=@n15)) {layerGroup=None}
  ]}
 EventGraph {nodes [
   Event BlueprintUpdateAnimation()
   n1 = BreakTransform(InTransform=GetDeltaTransformFromRefPose(self=GetOwnerSkeletalMeshComponent(), BoneName=thigh_r, BaseName=calf_r))
   > Set calf_rot_r(calf_rot_r:double=MapRangeClamped(Value=BreakRotator(InRot=@n1.Rotation).Yaw, InRangeA=-20, InRangeB=-110)) -> (Output_Get:double)
   n2 = BreakTransform(InTransform=GetDeltaTransformFromRefPose(self=GetOwnerSkeletalMeshComponent(), BoneName=thigh_l, BaseName=calf_l))
   > Set calf_rot_l(calf_rot_l:double=MapRangeClamped(Value=BreakRotator(InRot=@n2.Rotation).Yaw, InRangeA=-10, InRangeB=-120)) -> (Output_Get:double)
   > n3 = Branch(Condition=Get bIsSkyDiving())
   > Set Is Sky Diving(Is Sky Diving:double=0.15) -> (Output_Get:double)
   Set Is Sky Diving(execute=@n3.else, Is Sky Diving=1)
   n4 = GetDeltaTransformFromRefPose(self=GetOwnerSkeletalMeshComponent(), BoneName=thigh_r, BaseName=pelvis)
   MapRangeClamped(Value=BreakRotator(InRot=BreakTransform(InTransform=@n4).Rotation).Yaw, InRangeA=-20, InRangeB=-100)
   n5 = GetDeltaTransformFromRefPose(self=GetOwnerSkeletalMeshComponent(), BoneName=thigh_l, BaseName=pelvis)
   MapRangeClamped(Value=BreakRotator(InRot=BreakTransform(InTransform=@n5).Rotation).Yaw, InRangeA=-20, InRangeB=-100)
  ]}
]
variables [
 calf_rot_l {type=number}
 calf_rot_r {type=number}
 Is Sky Diving {type=number}
]
